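# Control Rig for the NotchGnarl cosmetic backpack, whose skeleton has rocket, gun, hat, pole and flag bones in base/dyn pairs.
# On execution, two chained ModifyTransforms nodes apply fixed full-weight offsets to bones including the hat and flag.
# A ChainHarmonicsPerItem node then adds procedural secondary motion from base_pole down the flag chain, using pendulum and noisy wave motion with reach disabled.
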ControlRigBlueprint /BRCosmetics/Accessories/FORT_Backpacks/Backpack_M_MED_NotchGnarl/Meshes/M_MED_NotchGnarl_Pack_CtrlRig
graphs [
 RigVMFunctionLibrary {links [], nodes []}
 RigVMModel {nodes [
   RigUnit_BeginExecution = &RigUnit_BeginExecution()
   &RigUnit_ModifyTransforms(
    ItemToModify="((Item=(Type=Bone,Name=\"base_hat\"),Transform=(Rotation=(X=0.0436,Y=-0.0262,Z=0.0011,W=0.9987),Translation=(X=0,Y=0,Z=0.1),Scale3D=(X=1,Y=1,Z=1))),(Item=(Type=Bone,Name=\"base_flag_01\"),Transform=(Rotation=(X=0,Y=-0.0436,Z=0,W=0.999),Translation=(X=0,Y=0,Z=0 ...(truncated)"
    Weight=1
    WeightMinimum=0
    ExecutePin=@RigUnit_BeginExecution.ExecutePin
   )
   > &RigUnit_ModifyTransforms(
    ItemToModify="((Item=(Type=Bone,Name=\"base_flag_01\"),Transform=(Rotation=(X=0,Y=0,Z=0,W=1),Translation=(X=0,Y=0,Z=0),Scale3D=(X=1,Y=1,Z=1))),(Item=(Type=Bone,Name=\"dyn_flag_02\"),Transform=(Rotation=(X=0.2586,Y=0.0421,Z=-0.0113,W=0.965),Translation=(X=-0.25,Y=0,Z=0),Scal ...(truncated)"
    Weight=1
    WeightMinimum=0
   )
   > /Script/ControlRig.RigUnit_ChainHarmonicsPerItem(
    ChainRoot=(Type=Bone,Name="base_pole")
    DrawWorldOffset=(Rotation=(X=0,Y=0,Z=0,W=1),Translation=(X=0,Y=0,Z=0),Scale3D=(X=1,Y=1,Z=1))
    Pendulum=(bEnabled=true,PendulumStiffness=0.5,PendulumGravity=(X=0,Y=0,Z=20),PendulumBlend=0.25,PendulumDrag=1.5,PendulumMinimum=0,PendulumMaximum=0.1,PendulumEase=ElasticEaseInOut,UnwindAxis=(X=0,Y=0,Z=1),UnwindMinimum=1,UnwindMaximum=100)
    Reach=(bEnabled=false,ReachTarget=(X=10,Y=-20,Z=10),ReachAxis=(X=1,Y=0,Z=0),ReachMinimum=0,ReachMaximum=0.2,ReachEase=SineEaseIn)
    Speed=(X=8.5,Y=1.5,Z=5)
    Wave=(bEnabled=true,WaveFrequency=(X=25,Y=1,Z=5),WaveAmplitude=(X=0.25,Y=0.025,Z=0.05),WaveOffset=(X=10,Y=0,Z=5),WaveNoise=(X=0.1,Y=0.05,Z=0.05),WaveMinimum=-40,WaveMaximum=40,WaveEase=ExponentialEaseInOut)
    WaveCurve=(EditorCurveData=(Keys=((InterpMode=RCIM_Cubic,TangentMode=RCTM_User,ArriveTangent=1.132,LeaveTangent=1.132),(InterpMode=RCIM_Cubic,TangentMode=RCTM_User,Time=1,Value=1,ArriveTangent=0.3679,LeaveTangent=0.3679))))
    bDrawDebug=false
   ) {is=struct RigVMUnitNode}
  ]}
]
hierarchy [
 root {loc=<0 0 0>, rot=<0 0 0 1>}
 pack {parent=root, loc=<0 -0.7316 0>, rot=<0 0 0 1>}
 base_rocket {parent=pack, loc=<0 -0.2271 -8.3611>, rot=<0 -0.7071 0 -0.7071>}
 dyn_rocket {parent=base_rocket, loc=<0 0 0>, rot=<0 0 0 1>}
 base_gun {parent=dyn_rocket, loc=<-2.2367 -5.9332 -10.4718>, rot=<0 0 0 1>}
 dyn_gun {parent=base_gun, loc=<0 0 0>, rot=<0 0 0 1>}
 base_gunSpring {parent=dyn_gun, loc=<14.1044 -4.3616 0>, rot=<0 0 0 1>}
 dyn_gunSpring {parent=base_gunSpring, loc=<0 0 0>, rot=<0 0 0 1>}
 dyn_gunSpringCurl {parent=dyn_gunSpring, loc=<0 -2.2132 0>, rot=<0 0 0 1>}
 base_hat {parent=dyn_rocket, loc=<-1.9124 -5.6522 7.8417>, rot=<0.1834 0 0 -0.983>}
 dyn_hat {parent=base_hat, loc=<0 0 0>, rot=<0 0 0 1>}
 base_hatSpring_01 {parent=dyn_hat, loc=<-0.8065 -0.9643 1.2314>, rot=<0.1966 -0.7994 -0.4136 0.3888>}
 dyn_hatSpring_01 {parent=base_hatSpring_01, loc=<0 0 0>, rot=<0 0 0 1>}
 dyn_hatSpring_02 {parent=dyn_hatSpring_01, loc=<3.4734 -0.0349 0.0509>, rot=<0.0412 -0.15 -0.0792 -0.9846>}
 dyn_hatSpring_03 {parent=dyn_hatSpring_02, loc=<2.6063 -0.046 0.1337>, rot=<-0.1844 -0.1224 -0.0411 -0.9743>}
 base_hatFront_01 {parent=dyn_hat, loc=<0.2815 -3.0717 1.5078>, rot=<-0.0286 0.0232 0.0126 -0.9992>}
 dyn_hatFront_01 {parent=base_hatFront_01, loc=<0 0 0>, rot=<0 0 0 1>}
 dyn_hatFront_02 {parent=dyn_hatFront_01, loc=<11.2666 0.0089 0.0337>, rot=<0.0277 0.0012 0 -0.9996>}
 base_hatBack_01 {parent=dyn_hat, loc=<-0.0173 0.677 1.6652>, rot=<-0.4435 0.033 -0.0665 -0.8932>}
 dyn_hatBack_01 {parent=base_hatBack_01, loc=<0 0 0>, rot=<0.4248 -0.0384 -0.0074 -0.9044>}
 dyn_hatBack_02 {parent=dyn_hatBack_01, loc=<11.2666 0 0>, rot=<0 0 0 1>}
 base_hatMid_01 {parent=dyn_hat, loc=<0.9361 -1.8044 3.0637>, rot=<-0.112 0.1 0.0235 -0.9884>}
 dyn_hatMid_01 {parent=base_hatMid_01, loc=<0 0 0>, rot=<0 0 0 1>}
 base_pole {parent=dyn_rocket, loc=<-23.9741 -9.0567 7.4424>, rot=<-0.2536 0 -0.0151 -0.9672>}
 dyn_pole {parent=base_pole, loc=<0 0 0>, rot=<0 0 0 1>}
 base_flag_01 {parent=dyn_pole, loc=<5.0117 0 0.7491>, rot=<0 0 0 1>}
 dyn_flag_01 {parent=base_flag_01, loc=<0 0 0>, rot=<0 0 0 1>}
 dyn_flag_02 {parent=dyn_flag_01, loc=<1.4874 -0.0988 3.6882>, rot=<0.137 0 0 -0.9906>}
 dyn_flag_03 {parent=dyn_flag_02, loc=<2.2985 0.0938 4.8099>, rot=<-0.157 0 0 -0.9876>}
]
