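# Physics asset for the NotchGnarl backpack cosmetic, giving collision bodies and joints to the pack and its flag.
# A large root capsule covers the pack, and a tiny flag base capsule is joined to it with swing and twist locked.
# Three slim capsules chain from that base along the dyn_flag bones, with unlocked joints that share identical swing and twist angle settings, so the flag can bend.
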PhysicsAsset /BRCosmetics/Accessories/FORT_Backpacks/Backpack_M_MED_NotchGnarl/Meshes/M_MED_NotchGnarl_Pack_Physics
bodies {
 base_flag_01 [{shapeType=Capsule, length=0.505, radius=0.505, shapeName=base_flag_01_capsule, rotation=<rot 0 -90 0>}]
 dyn_flag_01 [{
   shapeType=Capsule
   length=8.732
   radius=1.2738
   shapeName=dyn_flag_01_capsule
   center=<0.9422 0.074 2.0765>
   rotation=<rot 0 89.671 -89.1978>
  }]
 dyn_flag_02 [{
   shapeType=Capsule
   length=8.3815
   radius=1.6243
   shapeName=dyn_flag_01_capsule
   center=<0.9422 0.074 2.0765>
   rotation=<rot 0 89.671 -89.1978>
  }]
 dyn_flag_03 [{
   shapeType=Capsule
   length=8.5018
   radius=1.504
   shapeName=dyn_flag_01_capsule
   center=<0.9422 0.074 2.0765>
   rotation=<rot 0 89.671 -89.1978>
  }]
 root [{shapeType=Capsule, length=30.6907, radius=11.7455, shapeName=root_capsule, center=<0 2.9258 5.3716>, rotation=<rot 0 -90 90>}]
}
constraints [
 {
  bone1Name=base_flag_01
  bone2Name=root
  swing1LimitDegrees=45
  swing1Motion=Locked
  swing2LimitDegrees=45
  swing2Motion=Locked
  twistLimitDegrees=45
  twistMotion=Locked
 }
 {bone1Name=dyn_flag_01, bone2Name=base_flag_01, swing1LimitDegrees=30, swing2LimitDegrees=10, twistLimitDegrees=90}
 {bone1Name=dyn_flag_02, bone2Name=dyn_flag_01, swing1LimitDegrees=30, swing2LimitDegrees=10, twistLimitDegrees=90}
 {bone1Name=dyn_flag_03, bone2Name=dyn_flag_02, swing1LimitDegrees=30, swing2LimitDegrees=10, twistLimitDegrees=90}
]
